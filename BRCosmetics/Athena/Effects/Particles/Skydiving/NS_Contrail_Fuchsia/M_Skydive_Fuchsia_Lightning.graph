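# Translucent two-sided material for lightning strands in the Fuchsia skydiving contrail, used on Niagara mesh particles.
# World position offset combines SplineThicken width, either noise displacement or sine wobble, and a bend toward a dynamic-parameter position, all masked by vertex red.
# Emissive blends edge and core colors across the strand, adds scrolling noise, then boosts to HDR.
# Opacity combines the strand shape, lengthwise fade, particle alpha, dynamic-parameter dissolve and depth fade.
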Material /BRCosmetics/Athena/Effects/Particles/Skydiving/NS_Contrail_Fuchsia/M_Skydive_Fuchsia_Lightning
graph {
 properties {
  bUsedWithNiagaraMeshParticles=true
  blendMode=BLEND_Translucent
  pixelDepthOffsetMode=PDOM_Legacy
  twoSided=true
  usedShadingModels=Default Lit
 }
 nodes [
  n1 = DynamicParameter() {paramNames [Width,WobbleMult,Random1,Random2]}
  n2 = Multiply(A=Add(A=@n1.Random2, B=Time()), B=ScalarParameter() {defaultValue=3, parameterName=WobbleSpeed})
  n3 = VectorParameter() {parameterName=SineWaveMul, defaultValue=<rgba -0.6 -1 0.2 0>}
  n4 = ComponentMask(None=Add(A=TextureCoordinate(), B=Multiply(A=@n2, B=@n3.R) {constB=-0.6})) {g=true}
  n5 = Sine(None=@n4)
  Add(A=@n5) {constB=0.5}
  n6 = VertexColor()
  n7 = Multiply(A=ScalarParameter() {defaultValue=6, parameterName=Width}, B=@n1.Width)
  n8 = ScalarParameter() {parameterName=WidthTipMult}
  n9 = TextureCoordinate()
  n10 = LinearInterpolate(B=@n1.Random1, Alpha=ScalarParameter() {parameterName=Use Random})
  n11 = Add(A=TextureCoordinate(), B=@n10) {constB=0}
  n12 = ComponentMask(None=VectorParameter() {group=WPO, parameterName=WPO_UVScale, defaultValue=<rgba 0 1.2 0 0>}.RGB) {g=true, r=true}
  n13 = ComponentMask(None=VectorParameter() {group=WPO, parameterName=WPO_Speed, defaultValue=<rgba 0 0.3 0 0>}.RGB) {g=true, r=true}
  n14 = TextureSample(UVs=Panner(Coordinate=Multiply(A=@n11, B=@n12), Time=@n2, Speed=@n13) {speedY=0.5}) {samplerType=SAMPLERTYPE_LinearColor, texture=&T_NoisePack_LinearColor}
  n15 = TextureCoordinate()
  n16 = VectorParameter() {group=WPO, parameterName=WPO_UVScale002, defaultValue=<rgba 0 1 0 0>}
  n17 = Panner(Coordinate=Multiply(A=Add(A=@n15, B=@n10) {constB=0}, B=ComponentMask(None=@n16.RGB) {g=true, r=true}), Time=@n2, Speed=@n13) {speedY=0.5}
  n18 = ScalarParameter() {defaultValue=0.5, parameterName=WPO Blend Min}
  n19 = ScalarParameter() {defaultValue=4, parameterName=WPO Blend Max}
  n20 = ScalarParameter() {defaultValue=0.3, parameterName=Blend offset}
  n21 = Add(A=Add(A=TextureCoordinate(), B=@n20) {constB=0.2}) {constB=0}
  n22 = ComponentMask(None=VectorParameter() {group=WPO, parameterName=WPO_Blend_UVScale, defaultValue=<rgba 0 1 0 0>}.RGB) {g=true, r=true}
  n23 = TextureSample(UVs=Panner(Coordinate=Multiply(A=@n21, B=@n22), Time=@n2, Speed=@n13) {speedY=0.3}) {samplerType=SAMPLERTYPE_LinearColor, texture=&T_NoisePack_LinearColor}
  n24 = ScalarParameter() {parameterName=WPO Blend Min Mask}
  n25 = ScalarParameter() {defaultValue=1.2, parameterName=WPO Blend Max Mask}
  n26 = Add(A=TextureCoordinate() {uTiling=0}, B=@n20) {constB=0.3}
  n27 = VectorParameter() {group=WPO, parameterName=WPO Blend mask Speed, defaultValue=<rgba 0 0.4 0 0>}
  n28 = TextureSample(UVs=Panner(Coordinate=Add(A=@n26) {constB=0}, Time=@n2, Speed=ComponentMask(None=@n27.RGB) {g=true, r=true}) {speedY=0.4}) {samplerType=SAMPLERTYPE_Normal, texture=&T_Weap_Distort_N_02}
  n29 = Power(
   Base=Multiply(A=LinearInterpolate(A=@n24, B=@n25, Alpha=@n28.R)) {constB=2}
   Exp=ScalarParameter() {defaultValue=1.5, parameterName=WPO Blend Power}
  ) {constExponent=1.5}
  n30 = Power(Base=ComponentMask(None=TextureCoordinate()) {g=true}) {constExponent=1}
  n31 = Multiply(
   A=Multiply(A=LinearInterpolate(A=@n18, B=@n19, Alpha=@n23.R) {constA=0.5, constB=4}, B=Multiply(A=@n29) {constB=0.5}) {constB=0.5}
   B=Saturate(None=Multiply(A=@n30) {constB=7})
  )
  n32 = ComponentMask(None=TextureCoordinate()) {g=true}
  n33 = LinearInterpolate(Alpha=Saturate(None=Power(Base=@n32) {constExponent=4})) {constA=1, constB=1.5}
  n34 = Add(
   A=LinearInterpolate(
    A=LinearInterpolate(Alpha=@n14.G)
    B=TextureSample(UVs=@n17) {samplerType=SAMPLERTYPE_Normal, texture=&T_Weap_Distort_N_02}.R
   ) {constAlpha=0.6}
   B=LinearInterpolate(B=Multiply(A=@n31, B=@n33), Alpha=ScalarParameter() {defaultValue=0.8, parameterName=Add blend})
  ) {constB=0}
  n35 = TextureCoordinate()
  n36 = Power(Base=OneMinus(None=ComponentMask(None=@n35) {g=true})) {constExponent=1}
  n37 = ScalarParameter() {defaultValue=5, parameterName=Bottom Mask}
  n38 = TextureCoordinate()
  n39 = Power(Base=Sine(None=ComponentMask(None=@n38) {g=true}) {period=2})
  n40 = LinearInterpolate(
   A=Saturate(None=Multiply(A=@n36, B=@n37) {constB=5})
   B=Saturate(None=Multiply(A=@n39) {constB=5})
   Alpha=ScalarParameter() {parameterName=Use both sides mask}
  )
  n41 = Multiply(A=Multiply(A=LinearInterpolate(Alpha=@n34) {constB=1.5}, B=@n40), B=Constant3Vector() {constant=<rgba 1 0 0 1>})
  n42 = TextureCoordinate() {uTiling=0}
  n43 = ParticleRandom()
  n44 = Add(A=@n42, B=@n43)
  n45 = Add(A=@n42, B=@n43)
  n46 = LinearInterpolate(
   A=TextureSample(UVs=Panner(Coordinate=@n44, Time=@n2) {speedY=0.5}) {samplerType=SAMPLERTYPE_Normal, texture=&T_Weap_Distort_N_02}.R
   B=TextureSample(UVs=Panner(Coordinate=@n45, Time=@n2) {speedY=0.5}) {samplerType=SAMPLERTYPE_LinearColor, texture=&T_NoisePack_LinearColor}.R
  )
  n47 = Multiply(
   A=Multiply(A=LinearInterpolate(Alpha=@n46) {constB=1.5}, B=@n40)
   B=VectorParameter() {parameterName=WPO Y, defaultValue=<rgba 0 0.5 0 1>}.RGB
  )
  n48 = Add(A=Transform(None=Add(A=@n41, B=@n47) {constB=0}) {transformSourceType=TRANSFORMSOURCE_Local}) {constB=0}
  n49 = Multiply(A=ScalarParameter() {defaultValue=30, parameterName=WPO Intensity})
  n50 = Constant() {r=0.5}
  n51 = SphereMask(A=ComponentMask(None=@n9) {g=true}, B=@n50, Radius=@n50) {hardnessPercent=50}
  n52 = Cosine(None=@n4)
  n53 = AppendVector(A=AppendVector(A=Constant(), B=Multiply(A=@n52, B=@n3.B) {constB=0.2}), B=Multiply(A=@n5, B=@n5))
  n54 = Multiply(
   A=Normalize(VectorInput=Transform(None=@n53) {transformSourceType=TRANSFORMSOURCE_Local})
   B=Multiply(A=ScalarParameter() {defaultValue=5, parameterName=Wobble}, B=@n1.WobbleMult)
  )
  n55 = Add(
   A=MaterialFunctionCall(WidthBase=@n7, WidthTip=Multiply(A=@n7, B=@n8) {constB=5}, UVs for Thickness=Power(Base=@n9) {constExponent=0.6}) {materialFunction=&SplineThicken}.WorldPositionOffset
   B=StaticSwitch(True=Multiply(A=@n48, B=@n49) {constB=30}, False=Multiply(A=@n51, B=@n54), Value=StaticBoolParameter() {parameterName=WAVE_SWITCH})
  )
  n56 = DynamicParameter() {parameterIndex=2, defaultValue=<rgba 1 1 100 0.01>, paramNames [X, Y, Z, Bend WPO Intensity]}
  n57 = Subtract(
   A=MaterialFunctionCall(X=@n56.X, Y=@n56.X, Z=@n56.X) {materialFunction=&MakeFloat3}.Result
   B=ParticlePositionWS()
  )
  n58 = Clamp(None=ComponentMask(None=@n57) {g=true, r=true}) {maxDefault=30, minDefault=-30}
  n59 = Clamp(None=ComponentMask(None=@n57) {b=true}) {maxDefault=200, minDefault=-200}
  n60 = ComponentMask(None=TextureCoordinate()) {g=true}
  n61 = Subtract(A=Power(Base=@n60, Exp=ScalarParameter() {defaultValue=4, group=WPO, parameterName=WPOBendPower})) {constB=0.1}
  n62 = Saturate(None=Add(A=@n61, B=@n61))
  n63 = StaticSwitchParameter(True=Multiply(A=AppendVector(A=@n58, B=@n59), B=Multiply(A=@n62, B=@n56.Bend WPO Intensity)), False=Constant()) {defaultValue=true, group=WPO, parameterName=WPO From Velocity}
  n64 = Multiply(
   A=Constant3Vector() {constant=<rgba 0 0 100 0>}
   B=DynamicParameter() {parameterIndex=3, paramNames [FrontEnd_Z_Adjustment, 1, 3, 4]}.FrontEnd_Z_Adjustment
  )
  Add(A=Multiply(A=@n6.R, B=Add(A=@n55, B=@n63)), B=Multiply(A=@n6.R, B=Multiply(A=@n64, B=@n62)))
  Add(A=Sine(None=@n4) {period=0.6}) {constB=0.4}
  Clamp(None=@n53) {minDefault=-1}
  Constant() {r=1}
  n65 = ParticleColor()
  n66 = VectorParameter() {parameterName=EdgeColor, defaultValue=<rgba 0.2226 0.2425 0.5 0>}
  n67 = OneMinus(None=ComponentMask(None=TextureCoordinate()) {g=true})
  n68 = ScalarParameter() {defaultValue=1, parameterName=GradientIntensityScale}
  n69 = VectorParameter() {parameterName=Color, defaultValue=<rgba 0.6214 0.9501 1 0>}
  n70 = StaticSwitch(True=Multiply(A=Multiply(A=@n67, B=@n68), B=@n69.RGB), False=@n69.RGB, Value=StaticBoolParameter() {parameterName=AddGradient})
  n71 = ComponentMask(None=TextureCoordinate()) {r=true}
  n72 = Abs(None=Multiply(A=Add(A=@n71) {constB=-0.5}) {constB=2})
  n73 = Saturate(None=Multiply(A=OneMinus(None=@n72)) {constB=2})
  n74 = StaticSwitch(
   True=TextureSample() {samplerType=SAMPLERTYPE_LinearColor, texture=/BRCosmetics/Weapons/FORT_Melee/Pickaxe_Fuchsia_Female/VFX/T_Pickaxe_Fuchsia_Lightning}.R
   False=Multiply(A=@n73, B=@n73)
   Value=StaticBoolParameter() {parameterName=USE_TEXTURE}
  )
  n75 = VectorParameter() {parameterName=Color2, defaultValue=<rgba 0 0.8819 1 0>}
  n76 = TextureSample(UVs=Subtract(A=TextureCoordinate() {uTiling=0.2, vTiling=3}, B=@n2)) {texture=/Game/Effects/FogVolume/noise_flamy_Tile}
  n77 = Multiply(A=@n76.R, B=@n76.R)
  n78 = Multiply(A=@n77, B=@n77)
  n79 = Multiply(A=@n65.RGB, B=Add(A=LinearInterpolate(A=@n66.RGB, B=@n70, Alpha=@n74), B=Multiply(A=@n75.RGB, B=@n78)))
  n80 = MaterialFunctionCall(LDR Input=Clamp(None=@n79), HDR Tint=@n79, HDR Intensity Multiplier=ScalarParameter() {defaultValue=20, parameterName=HDRIntensity}) {materialFunction=&DeriveHDRfromLDR}
  n81 = OneMinus(None=ComponentMask(None=TextureCoordinate()) {g=true})
  n82 = StaticBoolParameter() {parameterName=AlphaGradient}
  MaterialFunctionCall(Emissive=StaticSwitch(True=Multiply(A=@n80.RemappedRange, B=@n81), False=@n80.RemappedRange, Value=@n82)) {materialFunction=&MF_FX_EmissiveAdjustments}
  OneMinus(None=@n60)
  Sine(None=ComponentMask(None=TextureCoordinate()) {g=true})
  n83 = ComponentMask(None=TextureCoordinate()) {g=true}
  n84 = DynamicParameter() {parameterIndex=1, paramNames [Dissolve,1,2,3]}
  n85 = Subtract(A=@n83, B=@n84.Dissolve)
  n86 = Constant() {r=1}
  n87 = Constant()
  n88 = If(A=OneMinus(None=Abs(None=@n85)), B=Constant() {r=0.8}, "A > B"=@n86, "A < B"=@n87)
  n89 = If(A=Add(A=OneMinus(None=@n84.Dissolve), B=@n83), B=Constant() {r=0.9999}, "A > B"=@n87, "A < B"=@n86)
  n90 = StaticBoolParameter() {parameterName=Thresuhold}
  n91 = OneMinus(None=@n83)
  n92 = Multiply(A=Add(A=@n74, B=Multiply(A=@n74) {constB=0.2}), B=@n65.A)
  n93 = DepthFade(
   Opacity=Multiply(A=StaticSwitch(True=@n88, False=@n89, Value=@n90), B=Multiply(A=@n91, B=@n92))
   FadeDistance=ScalarParameter() {defaultValue=30, parameterName=FadeDistance}
  )
  StaticSwitch(True=Multiply(A=@n93, B=@n81), False=@n93, Value=@n82)
  TextureCoordinate() {uTiling=0, vTiling=1.2}
  WorldPosition()
 ]
}
