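# Niagara system for the Fuchsia skydiving contrail cosmetic, layering pink-purple bowl meshes, sprite decals and electric arc meshes on the character.
# Emitters follow the user-supplied skeletal mesh through bone sampling, with several arc emitters disabled.
# System diving attributes and an integrator drive dive-based offset, scale, opacity and material changes.
# Under-body arcs feed a smoothed world position into dynamic material parameters, with separate tuning for front-end preview and gameplay.
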NiagaraSystem /BRCosmetics/Athena/Effects/Particles/Skydiving/NS_Contrail_Fuchsia/NS_Skydive_Fuchsia
dependencies {
 usedDataInterfaces [&NiagaraDataInterfaceCurve, &NiagaraDataInterfaceColorCurve]
 usedDynamicInputs [
  &Divide_Float
  &MakeFloatFromInt
  &ReturnParticleIDAttributes
  &SetBoolByFloatComparison
 ]
 usedModules [
  &SystemState
  /Game/Effects/Niagara/Modules/Skydiving/FN_DivingAttributes/v4/FN_DivingAttributes_v4
  &self:SystemSpawnScript.SystemScriptSource.SystemScriptGraph.NiagaraNodeAssignment_0.SetVariables_#1
  &FN_Integrator
 ]
}
emitters {
 Bowl_Multi {
  emitterSpawnScript [&FN_PreserveSkydiveUserVariables() {module=PreserveSkydiveUserVariables}]
  emitterUpdateScript [
   &EmitterState()
   &SpawnBurst_Instantaneous(Spawn Count=1, Use Spawn Probability=false) {enabled=false}
   &SpawnRate(SpawnRate=2, Use Spawn Probability=false)
  ]
  particleSpawnScript [
   &FN_SetSpawnGroup(Number of Spawn Groups=1) {enabled=false, module=FN_SetSpawnGroup001}
   /Niagara/Modules/Spawn/Initialization/InitializeParticle(
    Color=<rgba 3 0.6228 2.0597 1>
    Mesh Scale=<1.5 1.7 1.7>
    Module.Write Color=true
    Module.Write Lifetime=true
    Module.Write Mass=false
    Module.Write Material Random=true
    Module.Write Position=false
    Module.Write Scale=true
    Module.Write SpriteRotation=false
    Module.Write SpriteSize=false
    Position=&SimulationPosition()
    Sprite Rotation=&UniformRangedFloat(Minimum=0, Maximum=1)
   )
   &InitialMeshOrientation(
    Orientation Axis=<0 0 1>
    Rotation=&MakeVector(X=&RandomRangeFloat(Minimum=-1, Maximum=1), Y=0, Z=0)
    Use Orientation Vector=true
    Use Rotation Vector=true
   )
  ]
  particleUpdateScript [
   /Game/Effects/Niagara/Modules/Skydiving/FN_DivingAttributes/FN_DivingAttributes_v2(Accumulate Dive=0.1, Accumulate Rest=0.1, SkeletalMesh:NiagaraDataInterfaceSkeletalMesh=$User.SkeletalMesh) {module=FN_DivingAttributes_v2001}
   &SkeletalMeshLocation(
    Bone / Socket Index=4
    Bone Sampling Mode=::NewEnumerator4
    Inherit Velocity=false
    Interpolate Mesh Sampling=false
    MaskSpawnGroup=false
    Sampled Position Offset=&Lerp_Vector(A=<0 0 -5>, B=<0 0 -65>, Alpha=$System.FN_Integrator001.Value)
    Skeletal Mesh=$User.SkeletalMesh
   )
   &ParticleState()
   &SolveForcesAndVelocity(Clamp Velocity=false, Limit Acceleration=false)
   &ScaleColor(
    Scale Alpha=&FloatFromCurve(
     FloatCurve=DataInterface() {propertyValues {Curve {keys [
         {arriveTangent=4.0516, interpMode=RCIM_Cubic, leaveTangent=4.0516, tangentMode=RCTM_User, time=0, value=0}
         {arriveTangent=0.0961, interpMode=RCIM_Cubic, leaveTangent=0.0961, tangentMode=RCTM_User, time=0.25, value=0.5}
         {arriveTangent=0.2272, interpMode=RCIM_Cubic, leaveTangent=0.2272, tangentMode=RCTM_User, time=1, value=0}
        ]}}}
     Scale Curve=&Lerp_Float(A=1, B=2, Alpha=$System.FN_Integrator001.Value)
    )
    Scale RGB=&MakeVectorFromLinearColorRGB(Color=&ColorFromCurve(ColorCurve=DataInterface() {propertyValues {
        AlphaCurve {keys [<at 0 1 RCIM_Cubic>]}
        BlueCurve {keys [<at 0 1 RCIM_Cubic>,<at 0.2386 1 RCIM_Cubic>]}
        GreenCurve {keys [<at 0 0.3697 RCIM_Cubic>,<at 0.2386 0.2742 RCIM_Cubic>]}
        RedCurve {keys [<at 0 0.1628 RCIM_Cubic>,<at 0.2386 0.6238 RCIM_Cubic>]}
       }}))
    ScaleA=true
    ScaleRGB=true
   )
   &DynamicMaterialParameters(
    Index 0 Param 1=1
    Index 0 Param 2=1
    Index 0 Param 3=&FloatFromCurve(FloatCurve=DataInterface() {propertyValues {Curve {keys [{arriveTangent=1.3962, interpMode=RCIM_Cubic, leaveTangent=1.3962, tangentMode=RCTM_User, time=0.0032, value=0.136}, <at 1 1>]}}})
    Index 0 Param 4=&FloatFromCurve(
     FloatCurve=DataInterface() {propertyValues {Curve {keys [{arriveTangent=-1.0048, interpMode=RCIM_Cubic, leaveTangent=-1.0048, tangentMode=RCTM_User, time=-0.0046, value=1.0113}, <at 0.7 0.3>]}}}
     Scale Curve=0.5
    )
    Module.Param0WriteEnabled=false
    Module.Param1WriteEnabled=true
    Module.Param2WriteEnabled=true
    Module.Param3WriteEnabled=true
   )
   &ScaleMeshSize(Scale Factor=&VectorFromFloat(Value=&FloatFromCurve(
      FloatCurve=DataInterface() {propertyValues {Curve {keys [{arriveTangent=1.9256, interpMode=RCIM_Cubic, leaveTangent=1.9256, tangentMode=RCTM_User, time=0, value=0}, <at 1 1 RCIM_Cubic>]}}}
      Scale Curve=&Lerp_Float(A=0.9, B=1.4, Alpha=$System.FN_Integrator001.Value)
     )))
  ]
  rendererClasses [&NiagaraMeshRendererProperties]
  renderers [{rendererClass=&NiagaraMeshRendererProperties, rendererIndex=0}]
 }
 Decals {
  emitterUpdateScript [
   &EmitterState(Life Cycle Mode=::NewEnumerator1, Loop Delay=0.6, Loop Duration=0.5, UseLoopDelay=false)
   &SpawnBurst_Instantaneous(Spawn Count=2, Use Spawn Probability=false) {enabled=false}
   &SpawnRate(SpawnRate=15, Use Spawn Probability=false)
  ]
  particleSpawnScript [
   &FN_SetSpawnGroup(Number of Spawn Groups=2) {enabled=false}
   &InitializeParticle(
    Color=<rgba 0.5299 0.1406 3 1>
    Lifetime=&RandomRangeFloat(Minimum=0.5, Maximum=0.8)
    Module.Write Color=true
    Module.Write Lifetime=true
    Module.Write Mass=false
    Module.Write Material Random=true
    Module.Write Position=false
    Module.Write Scale=false
    Module.Write SpriteRotation=false
    Module.Write SpriteSize=true
    Position=&SimulationPosition()
    Sprite Size=&Vector2DFromFloat(Value=&FN_FrontEndFloatModifier(Gameplay=35, Front End=50))
   )
   &CameraOffset(Camera Offset Amount=&FN_FrontEndFloatModifier(Gameplay=25, Front End=5))
   /SetVariables(Particles.FloatID=HlslExpression() {hlslExpression=Particles.ID.Index}) {isSetParameters=true}
   /SetVariables(Particles.LR=HlslExpression() {hlslExpression="sign(fmod(Particles.FloatID,2) -0.5)"}) {isSetParameters=true}
  ]
  particleUpdateScript [
   &ParticleState()
   &SkeletalMeshLocation(
    Bone Sampling Mode=::NewEnumerator4
    Inherit Velocity=false
    MaskSpawnGroup=false
    Sampled Position Offset=&Multiply_Vector(
     A=<16 5 -1>
     B=&Add_Vector(
      A=&MakeVector(X=$Particles.LR, Y=1, Z=1)
      B=&FN_FrontEndVector3Modifier(
       Gameplay=&RandomRangeVector(Minimum=<0 -1 1>, Maximum=<0 1 1>)
       Front End=&RandomRangeVector(Minimum=<0 -1.5 1>, Maximum=<0 1.2 2>)
      )
     )
    )
    Sampled Position Offset Coordinate Space:ENiagaraMeshTransforms=::NewEnumerator2
    Skeletal Mesh=$User.SkeletalMesh
   )
   &DynamicMaterialParameters(
    Index 0 Param 1=&FloatFromCurve(
     FloatCurve=DataInterface() {propertyValues {Curve {keys [
         {arriveTangent=-1.656, interpMode=RCIM_Cubic, leaveTangent=-1.656, tangentMode=RCTM_User, time=0, value=1}
         <at 0.85 0 RCIM_Cubic>
         {arriveTangent=8.0037, interpMode=RCIM_Cubic, leaveTangent=8.0037, tangentMode=RCTM_User, time=1, value=1}
        ]}}}
     Scale Curve=3
    )
    Index 0 Param 2=1
    Index 0 Param 4=$Particles.MaterialRandom
    Module.Param0WriteEnabled=true
    Module.Param1WriteEnabled=false
    Module.Param2WriteEnabled=false
    Module.Param3WriteEnabled=false
   )
   &ScaleColor(
    Scale Alpha=&FloatFromCurve(
     FloatCurve=DataInterface() {propertyValues {Curve {keys [
         {interpMode=RCIM_Cubic, tangentMode=RCTM_User, time=0, value=0}
         {interpMode=RCIM_Cubic, tangentMode=RCTM_User, time=0.1, value=1}
         {interpMode=RCIM_Cubic, tangentMode=RCTM_User, time=0.9, value=1}
         {interpMode=RCIM_Cubic, tangentMode=RCTM_User, time=1, value=0}
        ]}}}
     Scale Curve=&FN_FrontEndFloatModifier(Gameplay=1, Front End=3)
    )
    Scale RGB=&Lerp_Vector(A=<0 0 0>, B=<1 0.1 0.1>, Alpha=$System.SwitchGlow)
    ScaleA=true
    ScaleRGB=false
   )
  ]
  rendererClasses [&NiagaraSpriteRendererProperties]
  renderers [{rendererClass=&NiagaraSpriteRendererProperties}]
 }
 ElectricMeshArcsBody {
  emitterUpdateScript [
   &EmitterState(Life Cycle Mode=::NewEnumerator1, Loop Behavior=::NewEnumerator1, Loop Duration=0, UseLoopDelay=false)
   &SpawnBurst_Instantaneous(Spawn Count=1, Use Spawn Probability=false)
   &SpawnRate(SpawnRate=&RandomRangeFloat(Minimum=0.3, Maximum=1.5), Use Spawn Probability=false) {enabled=false}
  ]
  particleSpawnScript [
   &FN_SetSpawnGroup(Number of Spawn Groups=1) {enabled=false}
   &InitializeParticle(
    Color=&UniformRangedLinearColor(Minimum=<rgba 0.2328 0.2287 1 1.5>, Maximum=<rgba 0.2823 0 1 1.5>)
    Lifetime=&UniformRangedFloat(Minimum=1, Maximum=1)
    Mesh Scale=&MakeVector(X=3, Y=2, Z=4)
    Module.Write Color=true
    Module.Write Lifetime=true
    Module.Write Mass=true
    Module.Write Material Random=true
    Module.Write Position=false
    Module.Write Scale=true
    Module.Write SpriteRotation=false
    Module.Write SpriteSize=false
    Sprite Size=<1 1>
   )
   &DynamicMaterialParameters(
    Index 0 Param 1=&UniformRangedFloat(Minimum=0.2, Maximum=0.1)
    Index 0 Param 2=&UniformRangedFloat(Minimum=0.4, Maximum=0.6)
    Index 0 Param 3=&UniformRangedFloat(Minimum=0, Maximum=1)
    Index 0 Param 4=&RandomRangeFloat(Minimum=0, Maximum=1)
    Index 1 Param 1=0.01
    Index 1 Param 2=&RandomRangeFloat(Minimum=0.01, Maximum=0)
    Module.Param0WriteEnabled=true
    Module.Param0WriteEnabled1=true
    Module.Param0WriteEnabled2=true
    Module.Param1WriteEnabled=true
    Module.Param1WriteEnabled1=true
    Module.Param1WriteEnabled2=true
    Module.Param2WriteEnabled=true
    Module.Param2WriteEnabled1=false
    Module.Param2WriteEnabled2=true
    Module.Param3WriteEnabled=true
    Module.Param3WriteEnabled1=false
    Module.Param3WriteEnabled2=true
    Write Parameter Index 1=true
    Write Parameter Index 2=true
   )
   &InitialMeshOrientation(Mesh Orientation Mode=::NewEnumerator7, Rotation=<0 0.7 -0.2>, Use Rotation Vector=true)
  ]
  particleUpdateScript [
   &ParticleState(Kill Particles When Lifetime Has Elapsed=false)
   &SkeletalMeshLocation(
    Apply Sampled Orientation to Mesh Particles=true
    Bone / Socket Index=8
    Bone Sampling Mode=::NewEnumerator4
    Inherit Velocity=false
    Interpolate Mesh Sampling=false
    MaskSpawnGroup=false
    Sampled Position Offset=<7 -4 0>
    Sampled Position Offset Coordinate Space:ENiagaraMeshTransforms=::NewEnumerator2
    Skeletal Mesh=$User.SkeletalMesh
   )
   &ScaleColor(
    Scale Alpha=&FloatFromCurve(FloatCurve=DataInterface() {propertyValues {Curve {keys [<at 0.75 1 RCIM_Cubic>,<at 1 0>]}}}, Scale Curve=0.5)
    Scale RGB=&MakeVectorFromLinearColorRGB(Color=<rgba 1 1 1 1>)
    ScaleA=true
    ScaleRGB=false
   ) {enabled=false}
   &ScaleMeshSize(Scale Factor=&MakeVector(
     X=&FloatFromCurve(FloatCurve=DataInterface() {propertyValues {Curve {keys [<at 0 0.7 RCIM_Cubic>, {arriveTangent=1, leaveTangent=1, time=0.25, value=1}, <at 1 1>]}}})
     Y=1
     Z=&FloatFromCurve(FloatCurve=DataInterface() {propertyValues {Curve {keys [<at 0.8 1 RCIM_Cubic>, {arriveTangent=-11.4579, interpMode=RCIM_Cubic, leaveTangent=-11.4579, tangentMode=RCTM_User, time=1, value=0.8}]}}})
    )) {enabled=false}
   /SetVariables(Particles.MeshOrientation=&MultiplyQuaternion(Quaternion A=$Output.SkeletalMeshLocation.SampledRotation, Quaternion B=$Particles.Initial.MeshOrientation)) {isSetParameters=true, module=SetVariables_#2}
   &DynamicMaterialParameters(
    Index 0 Param 1=&FloatFromCurve(FloatCurve=DataInterface() {propertyValues {Curve {keys [<at 0 0.2>,<at 0.1 1.125>,<at 0.3 1 RCIM_Cubic>,<at 1 0>]}}}, Scale Curve=0.6)
    Index 1 Param 1=&FloatFromCurve(
     FloatCurve=DataInterface() {propertyValues {Curve {keys [
         {arriveTangent=1.6139, interpMode=RCIM_Cubic, leaveTangent=1.6139, tangentMode=RCTM_User, time=0, value=0.01}
         {tangentMode=RCTM_User, time=0.999, value=1.18}
        ]}}}
     CurveIndex=$Emitter.Age
    )
    Index 2 Param 4=0.0001
    Module.Param0WriteEnabled=false
    Module.Param0WriteEnabled1=true
    Module.Param0WriteEnabled2=true
    Module.Param1WriteEnabled=false
    Module.Param1WriteEnabled1=false
    Module.Param1WriteEnabled2=true
    Module.Param2WriteEnabled=false
    Module.Param2WriteEnabled1=false
    Module.Param2WriteEnabled2=true
    Module.Param3WriteEnabled=false
    Module.Param3WriteEnabled1=false
    Module.Param3WriteEnabled2=true
    Write Parameter Index 1=true
    Write Parameter Index 2=true
   ) {module=DynamicMaterialParameters001}
   &CameraOffset(Camera Offset Amount=10)
  ]
  rendererClasses [&NiagaraMeshRendererProperties]
  renderers [{rendererClass=&NiagaraMeshRendererProperties, rendererIndex=0}]
 }
 ElectricMeshArcsBody001 {
  emitterUpdateScript [
   &EmitterState(Life Cycle Mode=::NewEnumerator1, Loop Behavior=::NewEnumerator1, Loop Duration=0, UseLoopDelay=false)
   &SpawnBurst_Instantaneous(Spawn Count=1, Use Spawn Probability=false)
   &SpawnRate(SpawnRate=&RandomRangeFloat(Minimum=0.3, Maximum=2), Use Spawn Probability=false) {enabled=false}
  ]
  particleSpawnScript [
   &FN_SetSpawnGroup(Number of Spawn Groups=1) {enabled=false}
   &InitializeParticle(
    Color=&UniformRangedLinearColor(Minimum=<rgba 0.2328 0.2287 1 1.5>, Maximum=<rgba 0.2823 0 1 1.5>)
    Lifetime=&UniformRangedFloat(Minimum=1, Maximum=1)
    Mesh Scale=&MakeVector(X=2.5, Y=2.5, Z=3)
    Module.Write Color=true
    Module.Write Lifetime=true
    Module.Write Mass=true
    Module.Write Material Random=true
    Module.Write Position=false
    Module.Write Scale=true
    Module.Write SpriteRotation=false
    Module.Write SpriteSize=false
    Sprite Size=<1 1>
   )
   &DynamicMaterialParameters(
    Index 0 Param 1=&UniformRangedFloat(Minimum=0.2, Maximum=0.1)
    Index 0 Param 2=&UniformRangedFloat(Minimum=0.4, Maximum=0.3)
    Index 0 Param 3=&UniformRangedFloat(Minimum=0, Maximum=1)
    Index 0 Param 4=&RandomRangeFloat(Minimum=0, Maximum=1)
    Index 1 Param 1=0.01
    Index 1 Param 2=&RandomRangeFloat(Minimum=0.01, Maximum=0)
    Module.Param0WriteEnabled=true
    Module.Param0WriteEnabled1=true
    Module.Param0WriteEnabled2=true
    Module.Param1WriteEnabled=true
    Module.Param1WriteEnabled1=true
    Module.Param1WriteEnabled2=true
    Module.Param2WriteEnabled=true
    Module.Param2WriteEnabled1=false
    Module.Param2WriteEnabled2=true
    Module.Param3WriteEnabled=true
    Module.Param3WriteEnabled1=false
    Module.Param3WriteEnabled2=true
    Write Parameter Index 1=true
    Write Parameter Index 2=true
   )
   &InitialMeshOrientation(Mesh Orientation Mode=::NewEnumerator7, Rotation=<0 -0.14 0.2>, Use Rotation Vector=true)
  ]
  particleUpdateScript [
   &ParticleState(Kill Particles When Lifetime Has Elapsed=false)
   &SkeletalMeshLocation(
    Apply Sampled Orientation to Mesh Particles=true
    Bone / Socket Index=9
    Bone Sampling Mode=::NewEnumerator4
    Inherit Velocity=false
    Interpolate Mesh Sampling=false
    MaskSpawnGroup=false
    Sampled Position Offset=<-0.4 0 10>
    Sampled Position Offset Coordinate Space:ENiagaraMeshTransforms=::NewEnumerator2
    Skeletal Mesh=$User.SkeletalMesh
   )
   &ScaleColor(
    Scale Alpha=&FloatFromCurve(FloatCurve=DataInterface() {propertyValues {Curve {keys [<at 0.75 1 RCIM_Cubic>,<at 1 0>]}}}, Scale Curve=0.5)
    Scale RGB=&MakeVectorFromLinearColorRGB(Color=<rgba 1 1 1 1>)
    ScaleA=false
    ScaleRGB=false
   )
   &ScaleMeshSize(Scale Factor=&MakeVector(
     X=&FloatFromCurve(FloatCurve=DataInterface() {propertyValues {Curve {keys [<at 0 0.7 RCIM_Cubic>, {arriveTangent=1, leaveTangent=1, time=0.25, value=1}, <at 1 1>]}}})
     Y=1
     Z=&FloatFromCurve(FloatCurve=DataInterface() {propertyValues {Curve {keys [<at 0.8 1 RCIM_Cubic>, {arriveTangent=-11.4579, interpMode=RCIM_Cubic, leaveTangent=-11.4579, tangentMode=RCTM_User, time=1, value=0.8}]}}})
    )) {enabled=false}
   /SetVariables(Particles.MeshOrientation=&MultiplyQuaternion(Quaternion A=$Output.SkeletalMeshLocation.SampledRotation, Quaternion B=$Particles.Initial.MeshOrientation)) {isSetParameters=true, module=SetVariables_#2}
   &DynamicMaterialParameters(
    Index 0 Param 1=&FloatFromCurve(FloatCurve=DataInterface() {propertyValues {Curve {keys [<at 0 0.2>,<at 0.1 1.125>,<at 0.3 1 RCIM_Cubic>,<at 1 0>]}}}, Scale Curve=0.6)
    Index 1 Param 1=&FloatFromCurve(
     FloatCurve=DataInterface() {propertyValues {Curve {keys [
         {arriveTangent=1.6139, interpMode=RCIM_Cubic, leaveTangent=1.6139, tangentMode=RCTM_User, time=0, value=0.01}
         {tangentMode=RCTM_User, time=0.999, value=1.18}
        ]}}}
     CurveIndex=$Emitter.Age
    )
    Index 2 Param 4=0.0001
    Module.Param0WriteEnabled=false
    Module.Param0WriteEnabled1=true
    Module.Param0WriteEnabled2=true
    Module.Param1WriteEnabled=false
    Module.Param1WriteEnabled1=false
    Module.Param1WriteEnabled2=true
    Module.Param2WriteEnabled=false
    Module.Param2WriteEnabled1=false
    Module.Param2WriteEnabled2=true
    Module.Param3WriteEnabled=false
    Module.Param3WriteEnabled1=false
    Module.Param3WriteEnabled2=true
    Write Parameter Index 1=true
    Write Parameter Index 2=true
   ) {module=DynamicMaterialParameters001}
   &CameraOffset(Camera Offset Amount=14)
  ]
  rendererClasses [&NiagaraMeshRendererProperties]
  renderers [{rendererClass=&NiagaraMeshRendererProperties, rendererIndex=0}]
 }
 ElectricMeshArcsUnder {
  bEnabled=false
  emitterUpdateScript [
   &EmitterState(Life Cycle Mode=::NewEnumerator1, Loop Duration=0, Recalculate Duration Each Loop=true, UseLoopDelay=false)
   &SpawnBurst_Instantaneous(Spawn Count=&UniformRangedInt(Minimum=0, Maximum=1), Use Spawn Probability=false) {enabled=false}
   &SpawnRate(SpawnRate=4.5, Use Spawn Probability=false)
  ]
  particleSpawnScript [
   &FN_SetSpawnGroup(Number of Spawn Groups=1) {enabled=false}
   /SetVariables(Particles.InitRand=HlslExpression() {hlslExpression="sign(rand(1.0)-0.5)"}) {isSetParameters=true, module=SetVariables_#3}
   &InitializeParticle(
    Color=&UniformRangedLinearColor(Minimum=<rgba 0.2328 0.2287 1 1>, Maximum=<rgba 0.2823 0 1 1>)
    Lifetime=&UniformRangedFloat(Minimum=0.5, Maximum=0.6)
    Mesh Scale=&Multiply_VectorByFloat(
     Vector=&MakeVector(
      X=4
      Y=&MaskFloatBySpawnGroup(Masked Float=3.5, Passthrough Float=4, Spawn Group Mask=&RandomRangeInt(Minimum=0, Maximum=1))
      Z=&MaskFloatBySpawnGroup(Masked Float=-5, Passthrough Float=5, Spawn Group Mask=&RandomRangeInt(Minimum=0, Maximum=1))
     )
     Float=&RandomRangeFloat(Minimum=0.85, Maximum=1.1)
    )
    Module.Write Color=true
    Module.Write Lifetime=true
    Module.Write Mass=true
    Module.Write Material Random=true
    Module.Write Position=false
    Module.Write Scale=true
    Module.Write SpriteRotation=false
    Module.Write SpriteSize=false
    Sprite Size=<1 1>
   )
   &DynamicMaterialParameters(
    Index 0 Param 1=&UniformRangedFloat(Minimum=0.5, Maximum=0.5)
    Index 0 Param 2=&UniformRangedFloat(Minimum=0.12, Maximum=0.232)
    Index 0 Param 3=&UniformRangedFloat(Minimum=0, Maximum=1)
    Index 1 Param 1=&UniformRangedFloat(Minimum=0, Maximum=1)
    Index 1 Param 2=&UniformRangedFloat(Minimum=0, Maximum=1)
    Module.Param0WriteEnabled=true
    Module.Param0WriteEnabled1=false
    Module.Param1WriteEnabled=true
    Module.Param1WriteEnabled1=true
    Module.Param2WriteEnabled=true
    Module.Param2WriteEnabled1=false
    Module.Param3WriteEnabled=false
    Module.Param3WriteEnabled1=false
    Write Parameter Index 1=true
   )
   &InitialMeshOrientation(
    Mesh Orientation Mode=::NewEnumerator7
    Rotation=&Multiply_Vector(A=&RandomRangeVector(Minimum=<-0.1 0.25 0>, Maximum=<0.2 0.25 0>), B=<1 1 1>)
    Rotation Coordinate Space=::NewEnumerator1
    Use Rotation Vector=true
   )
   &ApplyInitialForces()
  ]
  particleUpdateScript [
   &ParticleState()
   &SkeletalMeshLocation(
    Apply Sampled Orientation to Mesh Particles=true
    Bone / Socket Index=10
    Bone Sampling Mode=::NewEnumerator4
    Inherit Velocity=false
    Interpolate Mesh Sampling=false
    MaskSpawnGroup=false
    Sampled Position Offset Coordinate Space:ENiagaraMeshTransforms=::NewEnumerator2
    Skeletal Mesh=$User.SkeletalMesh
   )
   &ScaleColor(
    Scale Alpha=&FloatFromCurve(
     FloatCurve=DataInterface() {propertyValues {Curve {keys [{arriveTangent=0.421, interpMode=RCIM_Cubic, leaveTangent=0.421, tangentMode=RCTM_User, time=0.2, value=1}, <at 1 0>]}}}
     Scale Curve=8
    )
    Scale RGB=&MakeVectorFromLinearColorRGB(Color=<rgba 1 1 1 1>)
    ScaleA=true
    ScaleRGB=false
   )
   &self:CustomInterpolateOverTime(
    Define Initial Value=false
    Rate Of Change=4
    Target Vector Value:Vector3f=/Niagara/DynamicInputs/Transforms/TransformPosition(Position=<0 0 0>, Source Space=::Local, Destination Space=::World)
   ) {module=CustomInterpolateOverTime}
   &ScaleMeshSize(Scale Factor=&MakeVector(
     X=&FloatFromCurve(FloatCurve=DataInterface() {propertyValues {Curve {keys [<at 0 0.85 RCIM_Cubic>, {arriveTangent=1, leaveTangent=1, time=0.25, value=1}, <at 1 1>]}}})
     Y=1
     Z=&FloatFromCurve(FloatCurve=DataInterface() {propertyValues {Curve {keys [<at 0.8 1 RCIM_Cubic>, {arriveTangent=-5.8686, interpMode=RCIM_Cubic, leaveTangent=-5.8686, tangentMode=RCTM_User, time=1, value=0.7}]}}})
    ))
   /SetVariables(Particles.MeshOrientation=&MultiplyQuaternion(Quaternion A=$Output.SkeletalMeshLocation.SampledRotation, Quaternion B=$Particles.Initial.MeshOrientation)) {enabled=false, isSetParameters=true, module=SetVariables_#2}
   &DynamicMaterialParameters(
    Index 0 Param 1=&FloatFromCurve(FloatCurve=DataInterface() {propertyValues {Curve {keys [<at 0.3 1 RCIM_Cubic>,<at 1 0>]}}}, Scale Curve=0.6)
    Index 0 Param 2=&UniformRangedFloat(Minimum=0.12, Maximum=0.232)
    Index 1 Param 1=&FloatFromCurve(FloatCurve=DataInterface() {propertyValues {Curve {keys [{tangentMode=RCTM_User, time=0, value=0}, {tangentMode=RCTM_User, time=1, value=1}]}}})
    Index 2 Param 1=&MakeFloatFromVector(Vector=&Multiply_VectorByFloat(Vector=$Particles.Moving Average, Float=1), Channel=::NewEnumerator0)
    Index 2 Param 2=&MakeFloatFromVector(Vector=&Multiply_VectorByFloat(Vector=$Particles.Moving Average, Float=1), Channel=::NewEnumerator1)
    Index 2 Param 3=&MakeFloatFromVector(
     Vector=&MakeCustomVectorFromBool(Bool=$User.bIsFrontEndPreview, True Vector=<0 0 300>, False Vector=$Particles.Moving Average)
     Channel=::NewEnumerator2
    )
    Index 2 Param 4=0.5
    Module.Param0WriteEnabled=true
    Module.Param0WriteEnabled1=true
    Module.Param0WriteEnabled2=true
    Module.Param1WriteEnabled=false
    Module.Param1WriteEnabled1=false
    Module.Param1WriteEnabled2=true
    Module.Param2WriteEnabled=false
    Module.Param2WriteEnabled1=false
    Module.Param2WriteEnabled2=true
    Module.Param3WriteEnabled=false
    Module.Param3WriteEnabled1=false
    Module.Param3WriteEnabled2=true
    Write Parameter Index 1=true
    Write Parameter Index 2=true
   ) {module=DynamicMaterialParameters001}
   &CameraOffset(Camera Offset Amount=-4)
   &SolveForcesAndVelocity(Clamp Velocity=false, Limit Acceleration=false)
  ]
  rendererClasses [&NiagaraMeshRendererProperties]
  renderers [{rendererClass=&NiagaraMeshRendererProperties, rendererIndex=0}]
 }
 ElectricMeshArcsUnder001 {
  emitterUpdateScript [
   &EmitterState(Life Cycle Mode=::NewEnumerator1, Loop Behavior=::NewEnumerator1, Loop Duration=0, UseLoopDelay=false)
   &SpawnBurst_Instantaneous(Spawn Count=2, Use Spawn Probability=false)
   &SpawnRate(SpawnRate=4.5, Use Spawn Probability=false) {enabled=false}
  ]
  particleSpawnScript [
   &FN_SetSpawnGroup(Number of Spawn Groups=1) {enabled=false}
   /SetVariables(Particles.InitRand=HlslExpression() {hlslExpression="sign(rand(1.0)-0.5)"}) {isSetParameters=true, module=SetVariables_#3}
   &InitializeParticle(
    Color=&UniformRangedLinearColor(Minimum=<rgba 0.2328 0.2287 1 3>, Maximum=<rgba 0.2823 0 1 3>)
    Lifetime=&UniformRangedFloat(Minimum=0.5, Maximum=0.6)
    Mesh Scale=&Multiply_VectorByFloat(
     Vector=&MakeVector(
      X=4
      Y=&MaskFloatBySpawnGroup(Masked Float=3.5, Passthrough Float=4, Spawn Group Mask=&RandomRangeInt(Minimum=0, Maximum=1))
      Z=&MaskFloatBySpawnGroup(Masked Float=-5, Passthrough Float=5, Spawn Group Mask=&RandomRangeInt(Minimum=0, Maximum=1))
     )
     Float=&RandomRangeFloat(Minimum=0.78, Maximum=0.78)
    )
    Module.Write Color=true
    Module.Write Lifetime=true
    Module.Write Mass=true
    Module.Write Material Random=true
    Module.Write Position=false
    Module.Write Scale=true
    Module.Write SpriteRotation=false
    Module.Write SpriteSize=false
    Sprite Size=<1 1>
   )
   &DynamicMaterialParameters(
    Index 0 Param 1=&UniformRangedFloat(Minimum=0.5, Maximum=0.5)
    Index 0 Param 2=&UniformRangedFloat(Minimum=0.1, Maximum=0.1)
    Index 0 Param 3=&UniformRangedFloat(Minimum=0, Maximum=1)
    Index 1 Param 1=&UniformRangedFloat(Minimum=1, Maximum=1)
    Index 1 Param 2=&UniformRangedFloat(Minimum=0, Maximum=1)
    Module.Param0WriteEnabled=true
    Module.Param0WriteEnabled1=true
    Module.Param1WriteEnabled=true
    Module.Param1WriteEnabled1=true
    Module.Param2WriteEnabled=true
    Module.Param2WriteEnabled1=false
    Module.Param3WriteEnabled=true
    Module.Param3WriteEnabled1=false
    Write Parameter Index 1=true
   )
   &InitialMeshOrientation(
    Mesh Orientation Mode=::NewEnumerator7
    Rotation=&Multiply_Vector(A=&RandomRangeVector(Minimum=<-0.1 0.25 0>, Maximum=<0.2 0.25 0>), B=<1 1 1>)
    Rotation Coordinate Space=::NewEnumerator1
    Use Rotation Vector=true
   )
   &ApplyInitialForces()
  ]
  particleUpdateScript [
   &ParticleState(Kill Particles When Lifetime Has Elapsed=false)
   &SkeletalMeshLocation(
    Apply Sampled Orientation to Mesh Particles=true
    Bone / Socket Index=10
    Bone Sampling Mode=::NewEnumerator4
    Inherit Velocity=false
    Interpolate Mesh Sampling=false
    MaskSpawnGroup=false
    Sampled Position Offset=<0 5 0>
    Sampled Position Offset Coordinate Space:ENiagaraMeshTransforms=::NewEnumerator2
    Skeletal Mesh=$User.SkeletalMesh
   )
   &ScaleColor(
    Scale Alpha=8
    Scale RGB=&Multiply_VectorByFloat(
     Vector=<1 1 1>
     Float=&FloatFromCurve(FloatCurve=DataInterface() {propertyValues {Curve {keys [<at 0 1>,<at 0.2 8>,<at 1 1>]}}}, CurveIndex=$Emitter.Age)
    )
    ScaleA=false
    ScaleRGB=true
   )
   &self:CustomInterpolateOverTime(
    Define Initial Value=false
    Rate Of Change=4
    Target Vector Value:Vector3f=&TransformPosition(Position=<0 0 0>, Source Space=::Local, Destination Space=::World)
   ) {module=CustomInterpolateOverTime}
   &ScaleMeshSize(Scale Factor=&MakeVector(
     X=&FloatFromCurve(FloatCurve=DataInterface() {propertyValues {Curve {keys [<at 0 0.85 RCIM_Cubic>, {arriveTangent=1, leaveTangent=1, time=0.25, value=1}, <at 1 1>]}}})
     Y=1
     Z=&FloatFromCurve(FloatCurve=DataInterface() {propertyValues {Curve {keys [<at 0.8 1 RCIM_Cubic>, {arriveTangent=-5.8686, interpMode=RCIM_Cubic, leaveTangent=-5.8686, tangentMode=RCTM_User, time=1, value=0.7}]}}})
    )) {enabled=false}
   /SetVariables(Particles.MeshOrientation=&MultiplyQuaternion(Quaternion A=$Output.SkeletalMeshLocation.SampledRotation, Quaternion B=$Particles.Initial.MeshOrientation)) {enabled=false, isSetParameters=true, module=SetVariables_#2}
   &DynamicMaterialParameters(
    Index 1 Param 1=&FloatFromCurve(
     FloatCurve=DataInterface() {propertyValues {Curve {keys [{tangentMode=RCTM_User, time=0, value=0}, {tangentMode=RCTM_User, time=0.8, value=1}]}}}
     CurveIndex=$Emitter.Age
    )
    Index 2 Param 1=&MakeFloatFromVector(Vector=&Multiply_VectorByFloat(Vector=$Particles.Moving Average, Float=1.1), Channel=::NewEnumerator0)
    Index 2 Param 2=&MakeFloatFromVector(Vector=&Multiply_VectorByFloat(Vector=$Particles.Moving Average, Float=1.1), Channel=::NewEnumerator1)
    Index 2 Param 3=&MakeFloatFromVector(
     Vector=&MakeCustomVectorFromBool(
      Bool=$User.bIsFrontEndPreview
      True Vector=&VectorFromFloat(Value=&Lerp_Float(A=220, B=300, Alpha=$System.FN_Integrator001.Value))
      False Vector=$Particles.Moving Average
     )
     Channel=::NewEnumerator2
    )
    Index 2 Param 4=&FN_FrontEndFloatModifier(Gameplay=0.4, Front End=&Lerp_Float(A=-0.001, B=0, Alpha=$System.Dive))
    Index 3 Param 1=&FN_FrontEndFloatModifier(Gameplay=0, Front End=&Lerp_Float(A=0.14, B=1, Alpha=$System.FN_Integrator001.Value))
    Module.Param0WriteEnabled1=true
    Module.Param0WriteEnabled2=true
    Module.Param0WriteEnabled3=true
    Module.Param1WriteEnabled1=false
    Module.Param1WriteEnabled2=true
    Module.Param1WriteEnabled3=false
    Module.Param2WriteEnabled1=false
    Module.Param2WriteEnabled2=true
    Module.Param2WriteEnabled3=false
    Module.Param3WriteEnabled1=false
    Module.Param3WriteEnabled2=true
    Module.Param3WriteEnabled3=false
    Write Parameter Index 0=false
    Write Parameter Index 1=true
    Write Parameter Index 2=true
    Write Parameter Index 3=true
   ) {module=DynamicMaterialParameters001}
   &CameraOffset(Camera Offset Amount=-4)
   &SolveForcesAndVelocity(Clamp Velocity=false, Limit Acceleration=false)
  ]
  rendererClasses [&NiagaraMeshRendererProperties]
  renderers [
   {rendererClass=&NiagaraMeshRendererProperties, rendererIndex=0}
   {rendererClass=&NiagaraMeshRendererProperties, rendererIndex=1}
  ]
 }
 ElectricMeshArcsUnder002 {
  emitterUpdateScript [
   &EmitterState(Life Cycle Mode=::NewEnumerator1, Loop Behavior=::NewEnumerator1, Loop Duration=0, UseLoopDelay=false)
   &SpawnBurst_Instantaneous(Spawn Count=2, Use Spawn Probability=false)
   &SpawnRate(SpawnRate=4.5, Use Spawn Probability=false) {enabled=false}
  ]
  particleSpawnScript [
   &FN_SetSpawnGroup(Number of Spawn Groups=1) {enabled=false}
   /SetVariables(Particles.InitRand=HlslExpression() {hlslExpression="sign(rand(1.0)-0.5)"}) {isSetParameters=true, module=SetVariables_#3}
   &InitializeParticle(
    Color=&UniformRangedLinearColor(Minimum=<rgba 0.2328 0.2287 1 3>, Maximum=<rgba 0.2823 0 1 3>)
    Lifetime=&UniformRangedFloat(Minimum=0.5, Maximum=0.6)
    Mesh Scale=&Multiply_VectorByFloat(
     Vector=&MakeVector(
      X=-4
      Y=&MaskFloatBySpawnGroup(Masked Float=3.5, Passthrough Float=4, Spawn Group Mask=&RandomRangeInt(Minimum=0, Maximum=1))
      Z=&MaskFloatBySpawnGroup(Masked Float=-5, Passthrough Float=5, Spawn Group Mask=&RandomRangeInt(Minimum=0, Maximum=1))
     )
     Float=&RandomRangeFloat(Minimum=0.78, Maximum=0.78)
    )
    Module.Write Color=true
    Module.Write Lifetime=true
    Module.Write Mass=true
    Module.Write Material Random=true
    Module.Write Position=false
    Module.Write Scale=true
    Module.Write SpriteRotation=false
    Module.Write SpriteSize=false
    Sprite Size=<1 1>
   )
   &DynamicMaterialParameters(
    Index 0 Param 1=&UniformRangedFloat(Minimum=0.5, Maximum=0.5)
    Index 0 Param 2=&UniformRangedFloat(Minimum=0.1, Maximum=0.1)
    Index 0 Param 3=&UniformRangedFloat(Minimum=0, Maximum=1)
    Index 0 Param 4=0.56
    Index 1 Param 1=&UniformRangedFloat(Minimum=1, Maximum=1)
    Index 1 Param 2=&UniformRangedFloat(Minimum=0, Maximum=1)
    Module.Param0WriteEnabled=true
    Module.Param0WriteEnabled1=true
    Module.Param1WriteEnabled=true
    Module.Param1WriteEnabled1=true
    Module.Param2WriteEnabled=true
    Module.Param2WriteEnabled1=false
    Module.Param3WriteEnabled=true
    Module.Param3WriteEnabled1=false
    Write Parameter Index 1=true
   )
   &InitialMeshOrientation(
    Mesh Orientation Mode=::NewEnumerator7
    Rotation=&Multiply_Vector(A=&RandomRangeVector(Minimum=<0.1 0.25 0>, Maximum=<-0.2 0.25 0>), B=<1 1 1>)
    Rotation Coordinate Space=::NewEnumerator1
    Use Rotation Vector=true
   )
   &ApplyInitialForces()
  ]
  particleUpdateScript [
   &ParticleState(Kill Particles When Lifetime Has Elapsed=false)
   &SkeletalMeshLocation(
    Apply Sampled Orientation to Mesh Particles=true
    Bone / Socket Index=10
    Bone Sampling Mode=::NewEnumerator4
    Inherit Velocity=false
    Interpolate Mesh Sampling=false
    MaskSpawnGroup=false
    Sampled Position Offset=<0 5 0>
    Sampled Position Offset Coordinate Space:ENiagaraMeshTransforms=::NewEnumerator2
    Skeletal Mesh=$User.SkeletalMesh
   )
   &ScaleColor(
    Scale Alpha=8
    Scale RGB=&Multiply_VectorByFloat(
     Vector=<1 1 1>
     Float=&FloatFromCurve(FloatCurve=DataInterface() {propertyValues {Curve {keys [<at 0 1>,<at 0.2 8>,<at 1 1>]}}}, CurveIndex=$Emitter.Age)
    )
    ScaleA=false
    ScaleRGB=true
   )
   &self:CustomInterpolateOverTime(
    Define Initial Value=false
    Rate Of Change=4
    Target Vector Value:Vector3f=&TransformPosition(Position=<0 0 0>, Source Space=::Local, Destination Space=::World)
   ) {module=CustomInterpolateOverTime}
   &ScaleMeshSize(Scale Factor=&MakeVector(
     X=&FloatFromCurve(FloatCurve=DataInterface() {propertyValues {Curve {keys [<at 0 0.85 RCIM_Cubic>, {arriveTangent=1, leaveTangent=1, time=0.25, value=1}, <at 1 1>]}}})
     Y=1
     Z=&FloatFromCurve(FloatCurve=DataInterface() {propertyValues {Curve {keys [<at 0.8 1 RCIM_Cubic>, {arriveTangent=-5.8686, interpMode=RCIM_Cubic, leaveTangent=-5.8686, tangentMode=RCTM_User, time=1, value=0.7}]}}})
    )) {enabled=false}
   /SetVariables(Particles.MeshOrientation=&MultiplyQuaternion(Quaternion A=$Output.SkeletalMeshLocation.SampledRotation, Quaternion B=$Particles.Initial.MeshOrientation)) {enabled=false, isSetParameters=true, module=SetVariables_#2}
   &DynamicMaterialParameters(
    Index 1 Param 1=&FloatFromCurve(
     FloatCurve=DataInterface() {propertyValues {Curve {keys [{tangentMode=RCTM_User, time=0, value=0}, {tangentMode=RCTM_User, time=0.8, value=1}]}}}
     CurveIndex=$Emitter.Age
    )
    Index 2 Param 1=&MakeFloatFromVector(Vector=&Multiply_VectorByFloat(Vector=$Particles.Moving Average, Float=1.1), Channel=::NewEnumerator0)
    Index 2 Param 2=&MakeFloatFromVector(Vector=&Multiply_VectorByFloat(Vector=$Particles.Moving Average, Float=1.1), Channel=::NewEnumerator1)
    Index 2 Param 3=&MakeFloatFromVector(
     Vector=&MakeCustomVectorFromBool(
      Bool=$User.bIsFrontEndPreview
      True Vector=&VectorFromFloat(Value=&Lerp_Float(A=200, B=300, Alpha=$System.FN_Integrator001.Value))
      False Vector=$Particles.Moving Average
     )
     Channel=::NewEnumerator2
    )
    Index 2 Param 4=&FN_FrontEndFloatModifier(Gameplay=0.4, Front End=&Lerp_Float(A=-0.001, B=0, Alpha=$System.Dive))
    Index 3 Param 1=&FN_FrontEndFloatModifier(Gameplay=0, Front End=&Lerp_Float(A=0.14, B=1, Alpha=$System.FN_Integrator001.Value))
    Module.Param0WriteEnabled1=true
    Module.Param0WriteEnabled2=true
    Module.Param0WriteEnabled3=true
    Module.Param1WriteEnabled1=false
    Module.Param1WriteEnabled2=true
    Module.Param1WriteEnabled3=false
    Module.Param2WriteEnabled1=false
    Module.Param2WriteEnabled2=true
    Module.Param2WriteEnabled3=false
    Module.Param3WriteEnabled1=false
    Module.Param3WriteEnabled2=true
    Module.Param3WriteEnabled3=false
    Write Parameter Index 0=false
    Write Parameter Index 1=true
    Write Parameter Index 2=true
    Write Parameter Index 3=true
   ) {module=DynamicMaterialParameters001}
   &CameraOffset(Camera Offset Amount=-4)
  ]
  rendererClasses [&NiagaraMeshRendererProperties]
  renderers [
   {rendererClass=&NiagaraMeshRendererProperties, rendererIndex=0}
   {rendererClass=&NiagaraMeshRendererProperties, rendererIndex=1}
  ]
 }
 ElectricMeshArcsUnder003 {
  bEnabled=false
  emitterUpdateScript [
   &EmitterState(Life Cycle Mode=::NewEnumerator1, Loop Duration=0.2, Recalculate Duration Each Loop=true, UseLoopDelay=false)
   &SpawnBurst_Instantaneous(Spawn Count=&UniformRangedInt(Minimum=0, Maximum=1), Use Spawn Probability=false) {enabled=false}
   &SpawnRate(SpawnRate=&RandomRangeFloat(Minimum=0, Maximum=1.5), Use Spawn Probability=false)
  ]
  particleSpawnScript [
   &FN_SetSpawnGroup(Number of Spawn Groups=1) {enabled=false}
   /SetVariables(Particles.InitRand=HlslExpression() {hlslExpression="sign(rand(1.0)-0.5)"}) {isSetParameters=true, module=SetVariables_#3}
   &InitializeParticle(
    Color=&UniformRangedLinearColor(Minimum=<rgba 0.2328 0.2287 1 1>, Maximum=<rgba 0.2823 0 1 1>)
    Lifetime=&UniformRangedFloat(Minimum=0.8, Maximum=1)
    Mesh Scale=&Multiply_VectorByFloat(
     Vector=&MakeVector(
      X=4
      Y=&MaskFloatBySpawnGroup(Masked Float=3.8, Passthrough Float=4.22, Spawn Group Mask=&RandomRangeInt(Minimum=0, Maximum=1))
      Z=&MaskFloatBySpawnGroup(Masked Float=3, Passthrough Float=3, Spawn Group Mask=&RandomRangeInt(Minimum=0, Maximum=1))
     )
     Float=&RandomRangeFloat(Minimum=0.8, Maximum=1)
    )
    Module.Write Color=true
    Module.Write Lifetime=true
    Module.Write Mass=true
    Module.Write Material Random=true
    Module.Write Position=false
    Module.Write Scale=true
    Module.Write SpriteRotation=false
    Module.Write SpriteSize=false
    Sprite Size=<1 1>
   )
   &DynamicMaterialParameters(
    Index 0 Param 1=&UniformRangedFloat(Minimum=0.5, Maximum=0.5)
    Index 0 Param 2=&UniformRangedFloat(Minimum=0.04, Maximum=0.05)
    Index 0 Param 3=&UniformRangedFloat(Minimum=0, Maximum=1)
    Index 0 Param 4=&RandomRangeFloat(Minimum=0, Maximum=1)
    Index 1 Param 1=&UniformRangedFloat(Minimum=0, Maximum=1)
    Index 1 Param 2=&UniformRangedFloat(Minimum=0, Maximum=1)
    Module.Param0WriteEnabled=true
    Module.Param0WriteEnabled1=false
    Module.Param0WriteEnabled2=true
    Module.Param1WriteEnabled=true
    Module.Param1WriteEnabled1=true
    Module.Param1WriteEnabled2=true
    Module.Param2WriteEnabled=true
    Module.Param2WriteEnabled1=false
    Module.Param2WriteEnabled2=true
    Module.Param3WriteEnabled=true
    Module.Param3WriteEnabled1=false
    Module.Param3WriteEnabled2=true
    Write Parameter Index 1=true
    Write Parameter Index 2=true
   )
   &InitialMeshOrientation(
    Mesh Orientation Mode=::NewEnumerator7
    Rotation=&RandomRangeVector(Minimum=<-0.3 -1 -0.5>, Maximum=<-0.2 1 -0.5>)
    Rotation Coordinate Space=::NewEnumerator1
    Use Rotation Vector=true
   )
  ]
  particleUpdateScript [
   &ParticleState()
   &SkeletalMeshLocation(
    Apply Sampled Orientation to Mesh Particles=true
    Bone / Socket Index=10
    Bone Sampling Mode=::NewEnumerator4
    Inherit Velocity=false
    Interpolate Mesh Sampling=false
    MaskSpawnGroup=false
    Sampled Position Offset=<20 0 0>
    Sampled Position Offset Coordinate Space:ENiagaraMeshTransforms=::NewEnumerator2
    Skeletal Mesh=$User.SkeletalMesh
   )
   &ScaleColor(
    Scale Alpha=&FloatFromCurve(FloatCurve=DataInterface() {propertyValues {Curve {keys [{arriveTangent=0.421, interpMode=RCIM_Cubic, leaveTangent=0.421, tangentMode=RCTM_User, time=0.2, value=1}, <at 1 0>]}}})
    Scale RGB=&MakeVectorFromLinearColorRGB(Color=<rgba 1 1 1 1>)
    ScaleA=true
    ScaleRGB=false
   )
   &InterpolateOverTime(
    Rate Of Change=4
    Target Vector Value:Vector3f=&TransformPosition(Position=<0 0 0>, Source Space=::Local, Destination Space=::World)
    Variable Type=::NewEnumerator2
   )
   &ScaleMeshSize(Scale Factor=&MakeVector(
     X=&FloatFromCurve(
      FloatCurve=DataInterface() {propertyValues {Curve {keys [<at 0 0.85 RCIM_Cubic>, {arriveTangent=1, leaveTangent=1, time=0.25, value=1}, <at 1 1>]}}}
      Scale Curve=0.15
     )
     Y=0.15
     Z=&FloatFromCurve(
      FloatCurve=DataInterface() {propertyValues {Curve {keys [<at 0.8 1 RCIM_Cubic>, {arriveTangent=-5.8686, interpMode=RCIM_Cubic, leaveTangent=-5.8686, tangentMode=RCTM_User, time=1, value=0.7}]}}}
      Scale Curve=0.15
     )
    ))
   /SetVariables(Particles.MeshOrientation=&MultiplyQuaternion(Quaternion A=$Output.SkeletalMeshLocation.SampledRotation, Quaternion B=$Particles.Initial.MeshOrientation)) {enabled=false, isSetParameters=true, module=SetVariables_#2}
   &DynamicMaterialParameters(
    Index 0 Param 1=&FloatFromCurve(
     FloatCurve=DataInterface() {propertyValues {Curve {keys [{arriveTangent=-3.5888, interpMode=RCIM_Cubic, leaveTangent=-3.5888, tangentMode=RCTM_User, time=0.3, value=1}, <at 1 0>]}}}
     Scale Curve=0.6
    )
    Index 1 Param 1=&FloatFromCurve(FloatCurve=DataInterface() {propertyValues {Curve {keys [{tangentMode=RCTM_User, time=0, value=0}, {tangentMode=RCTM_User, time=1, value=1}]}}})
    Module.Param0WriteEnabled=true
    Module.Param0WriteEnabled1=true
    Module.Param1WriteEnabled=false
    Module.Param1WriteEnabled1=false
    Module.Param2WriteEnabled=false
    Module.Param2WriteEnabled1=false
    Module.Param3WriteEnabled=false
    Module.Param3WriteEnabled1=false
    Write Parameter Index 1=true
   ) {module=DynamicMaterialParameters001}
  ]
  rendererClasses [&NiagaraMeshRendererProperties]
  renderers [{rendererClass=&NiagaraMeshRendererProperties, rendererIndex=0}]
 }
 ElectricMeshArcsUnder004 {
  bEnabled=false
  emitterUpdateScript [
   &EmitterState(Life Cycle Mode=::NewEnumerator1, Loop Duration=0, Recalculate Duration Each Loop=true, UseLoopDelay=false)
   &SpawnBurst_Instantaneous(Spawn Count=&UniformRangedInt(Minimum=0, Maximum=1), Use Spawn Probability=false) {enabled=false}
   &SpawnRate(SpawnRate=4.5, Use Spawn Probability=false)
  ]
  particleSpawnScript [
   &FN_SetSpawnGroup(Number of Spawn Groups=1) {enabled=false}
   /SetVariables(Particles.InitRand=HlslExpression() {hlslExpression="sign(rand(1.0)-0.5)"}) {isSetParameters=true, module=SetVariables_#3}
   &InitializeParticle(
    Color=&UniformRangedLinearColor(Minimum=<rgba 0.2328 0.2287 1 1>, Maximum=<rgba 0.2823 0 1 1>)
    Lifetime=&UniformRangedFloat(Minimum=0.5, Maximum=0.6)
    Mesh Scale=&Multiply_VectorByFloat(
     Vector=&MakeVector(
      X=-4
      Y=&MaskFloatBySpawnGroup(Masked Float=3.5, Passthrough Float=4, Spawn Group Mask=&RandomRangeInt(Minimum=0, Maximum=1))
      Z=&MaskFloatBySpawnGroup(Masked Float=-5, Passthrough Float=5, Spawn Group Mask=&RandomRangeInt(Minimum=0, Maximum=1))
     )
     Float=&RandomRangeFloat(Minimum=0.85, Maximum=1.1)
    )
    Module.Write Color=true
    Module.Write Lifetime=true
    Module.Write Mass=true
    Module.Write Material Random=true
    Module.Write Position=false
    Module.Write Scale=true
    Module.Write SpriteRotation=false
    Module.Write SpriteSize=false
    Sprite Size=<1 1>
   )
   &DynamicMaterialParameters(
    Index 0 Param 1=&UniformRangedFloat(Minimum=0.5, Maximum=0.5)
    Index 0 Param 2=&UniformRangedFloat(Minimum=0.12, Maximum=0.232)
    Index 0 Param 3=&UniformRangedFloat(Minimum=0, Maximum=1)
    Index 1 Param 1=&UniformRangedFloat(Minimum=0, Maximum=1)
    Index 1 Param 2=&UniformRangedFloat(Minimum=0, Maximum=1)
    Module.Param0WriteEnabled=true
    Module.Param0WriteEnabled1=false
    Module.Param1WriteEnabled=true
    Module.Param1WriteEnabled1=true
    Module.Param2WriteEnabled=true
    Module.Param2WriteEnabled1=false
    Module.Param3WriteEnabled=false
    Module.Param3WriteEnabled1=false
    Write Parameter Index 1=true
   )
   &InitialMeshOrientation(
    Mesh Orientation Mode=::NewEnumerator7
    Rotation=&Multiply_Vector(A=&RandomRangeVector(Minimum=<0.1 0.25 0>, Maximum=<-0.2 0.25 0>), B=<1 1 1>)
    Rotation Coordinate Space=::NewEnumerator1
    Use Rotation Vector=true
   )
  ]
  particleUpdateScript [
   &ParticleState()
   &SkeletalMeshLocation(
    Apply Sampled Orientation to Mesh Particles=true
    Bone / Socket Index=10
    Bone Sampling Mode=::NewEnumerator4
    Inherit Velocity=false
    Interpolate Mesh Sampling=false
    MaskSpawnGroup=false
    Sampled Position Offset Coordinate Space:ENiagaraMeshTransforms=::NewEnumerator2
    Skeletal Mesh=$User.SkeletalMesh
   )
   &ScaleColor(
    Scale Alpha=&FloatFromCurve(
     FloatCurve=DataInterface() {propertyValues {Curve {keys [{arriveTangent=0.421, interpMode=RCIM_Cubic, leaveTangent=0.421, tangentMode=RCTM_User, time=0.2, value=1}, <at 1 0>]}}}
     Scale Curve=8
    )
    Scale RGB=&MakeVectorFromLinearColorRGB(Color=<rgba 1 1 1 1>)
    ScaleA=true
    ScaleRGB=false
   )
   &self:CustomInterpolateOverTime(
    Define Initial Value=false
    Rate Of Change=4
    Target Vector Value:Vector3f=&TransformPosition(Position=<0 0 0>, Source Space=::Local, Destination Space=::World)
   ) {module=CustomInterpolateOverTime}
   &ScaleMeshSize(Scale Factor=&MakeVector(
     X=&FloatFromCurve(FloatCurve=DataInterface() {propertyValues {Curve {keys [<at 0 0.85 RCIM_Cubic>, {arriveTangent=1, leaveTangent=1, time=0.25, value=1}, <at 1 1>]}}})
     Y=1
     Z=&FloatFromCurve(FloatCurve=DataInterface() {propertyValues {Curve {keys [<at 0.8 1 RCIM_Cubic>, {arriveTangent=-5.8686, interpMode=RCIM_Cubic, leaveTangent=-5.8686, tangentMode=RCTM_User, time=1, value=0.7}]}}})
    ))
   /SetVariables(Particles.MeshOrientation=&MultiplyQuaternion(Quaternion A=$Output.SkeletalMeshLocation.SampledRotation, Quaternion B=$Particles.Initial.MeshOrientation)) {enabled=false, isSetParameters=true, module=SetVariables_#2}
   &DynamicMaterialParameters(
    Index 0 Param 1=&FloatFromCurve(FloatCurve=DataInterface() {propertyValues {Curve {keys [<at 0.3 1 RCIM_Cubic>,<at 1 0>]}}}, Scale Curve=0.6)
    Index 0 Param 2=&UniformRangedFloat(Minimum=0.12, Maximum=0.232)
    Index 1 Param 1=&FloatFromCurve(FloatCurve=DataInterface() {propertyValues {Curve {keys [{tangentMode=RCTM_User, time=0, value=0}, {tangentMode=RCTM_User, time=1, value=1}]}}})
    Index 2 Param 1=&MakeFloatFromVector(Vector=&Multiply_VectorByFloat(Vector=$Particles.Moving Average, Float=1), Channel=::NewEnumerator0)
    Index 2 Param 2=&MakeFloatFromVector(Vector=&Multiply_VectorByFloat(Vector=$Particles.Moving Average, Float=1), Channel=::NewEnumerator1)
    Index 2 Param 3=&MakeFloatFromVector(
     Vector=&MakeCustomVectorFromBool(Bool=$User.bIsFrontEndPreview, True Vector=<0 0 300>, False Vector=$Particles.Moving Average)
     Channel=::NewEnumerator2
    )
    Index 2 Param 4=0.5
    Module.Param0WriteEnabled=true
    Module.Param0WriteEnabled1=true
    Module.Param0WriteEnabled2=true
    Module.Param1WriteEnabled=false
    Module.Param1WriteEnabled1=false
    Module.Param1WriteEnabled2=true
    Module.Param2WriteEnabled=false
    Module.Param2WriteEnabled1=false
    Module.Param2WriteEnabled2=true
    Module.Param3WriteEnabled=false
    Module.Param3WriteEnabled1=false
    Module.Param3WriteEnabled2=true
    Write Parameter Index 1=true
    Write Parameter Index 2=true
   ) {module=DynamicMaterialParameters001}
   &CameraOffset(Camera Offset Amount=-4)
  ]
  rendererClasses [&NiagaraMeshRendererProperties]
  renderers [{rendererClass=&NiagaraMeshRendererProperties, rendererIndex=0}]
 }
 Lig {
  emitterSpawnScript [&FN_PreserveSkydiveUserVariables() {module=PreserveSkydiveUserVariables}]
  emitterUpdateScript [
   &EmitterState(Life Cycle Mode=::NewEnumerator1, Loop Behavior=::NewEnumerator1, Loop Duration Mode=::NewEnumerator1, UseLoopDelay=false)
   &SpawnBurst_Instantaneous(Spawn Count=1, Use Spawn Probability=false) {enabled=false}
   &SpawnRate(SpawnRate=0, Use Spawn Probability=false) {enabled=false}
   &SpawnBurst_Instantaneous(Spawn Count=2, Use Spawn Probability=false, UseLoopCountLimit=false) {module=SpawnBurst_Instantaneous001}
  ]
  particleSpawnScript [
   &FN_SetSpawnGroup(Number of Spawn Groups=1) {enabled=false, module=FN_SetSpawnGroup001}
   &InitializeParticle(
    Color=<rgba 3 0.6228 2.0597 1>
    Mesh Scale=<1.5 1.7 1.7>
    Module.Write Color=true
    Module.Write Lifetime=true
    Module.Write Mass=false
    Module.Write Material Random=true
    Module.Write Position=false
    Module.Write Scale=false
    Module.Write SpriteRotation=false
    Module.Write SpriteSize=false
    Position=&SimulationPosition()
    Sprite Rotation=&UniformRangedFloat(Minimum=0, Maximum=1)
   )
   /SetVariables(Particles.FloatID=HlslExpression() {hlslExpression=Particles.ID.Index}) {isSetParameters=true}
   /SetVariables(Particles.OffsetID:NiagaraFloat=&Multiply_Float(A=$Particles.FloatID, B=0.5)) {isSetParameters=true}
  ]
  particleUpdateScript [
   &ParticleState(Kill Particles When Lifetime Has Elapsed=false)
   &SkeletalMeshLocation(
    Bone / Socket Index=10
    Bone Sampling Mode=::NewEnumerator4
    Inherit Velocity=false
    Interpolate Mesh Sampling=false
    MaskSpawnGroup=false
    Sampled Position Offset=&Multiply_Vector(
     A=<70 8 0>
     B=&MakeVector(X=&Add_Float(A=HlslExpression() {hlslExpression=Particles.ID}, B=-0.5), Y=1, Z=1)
    )
    Sampled Position Offset Coordinate Space:ENiagaraMeshTransforms=::NewEnumerator2
    Skeletal Mesh=$User.SkeletalMesh
   )
   &FN_DivingAttributes_v2(Accumulate Dive=0.1, Accumulate Rest=0.1, Bone Index=1, SkeletalMesh:NiagaraDataInterfaceSkeletalMesh=$User.SkeletalMesh) {module=FN_DivingAttributes_v2001}
   &SolveForcesAndVelocity(Clamp Velocity=false, Limit Acceleration=false)
   &ScaleColor(
    Scale Alpha=&FloatFromCurve(FloatCurve=DataInterface() {propertyValues {Curve {keys [
         {arriveTangent=4.0516, interpMode=RCIM_Cubic, leaveTangent=4.0516, tangentMode=RCTM_User, time=0, value=0}
         {arriveTangent=0.0961, interpMode=RCIM_Cubic, leaveTangent=0.0961, tangentMode=RCTM_User, time=0.25, value=0.5}
         {arriveTangent=0.2272, interpMode=RCIM_Cubic, leaveTangent=0.2272, tangentMode=RCTM_User, time=1, value=0}
        ]}}})
    Scale RGB=&MakeVectorFromLinearColorRGB(Color=&ColorFromCurve(ColorCurve=DataInterface() {propertyValues {
        AlphaCurve {keys [<at 0 1 RCIM_Cubic>]}
        BlueCurve {keys [<at 0 1 RCIM_Cubic>,<at 0.2386 1 RCIM_Cubic>]}
        GreenCurve {keys [<at 0 0.3697 RCIM_Cubic>,<at 0.2386 0.2742 RCIM_Cubic>]}
        RedCurve {keys [<at 0 0.1628 RCIM_Cubic>,<at 0.2386 0.6238 RCIM_Cubic>]}
       }}))
    ScaleA=true
    ScaleRGB=true
   ) {enabled=false}
   &FN_Light_Attributes(
    Color=&MultiplyLinearColor(
     Linear Color=<rgba 9.0397 4.7917 20 1>
     Scale Factor=&MakeVector4_FromFloat(Float=&Waveform(
       Waveform Count=::NewEnumerator2
       Global Phase=&Add_Float(A=$Emitter.Age, B=$Particles.OffsetID)
       "Waveform [1]"=::NewEnumerator0
       "[1] Phase"=&Add_Float(A=$Emitter.Age, B=$Particles.OffsetID)
       "[1] Frequency"=1
       "[1] Amplitude Min/Max"=<0 1>
       "[2] Average"=false
       "[2] Input Position"=0
       "[2] Phase"=&Add_Float(A=$Emitter.Age, B=$Particles.OffsetID)
       "[2] Phase Scale"=1
       "[2] Compound Frequency"=2
       "[2] Frequency"=2
       "[2] Amplitude Min/Max"=<0 1>
       "[2] Scale"=1
       "[2] Bias"=0
       "[2] Exponent"=1
       "[2] Pulse Duration"=0.1
       "[2] Clip Triangle Min/Max"=<0 0>
       "[2] Random Seed"=0
       "[3] Average"=false
       "[3] Input Position"=0
       "[3] Phase"=&Add_Float(A=$Emitter.Age, B=$Particles.OffsetID)
       "[3] Phase Scale"=0
       "[3] Frequency"=0
       "[3] Compound Frequency"=2
       "[3] Amplitude Min/Max"=<0 1>
       "[3] Scale"=1
       "[3] Bias"=0
       "[3] Exponent"=1
       "[3] Pulse Duration"=0.1
       "[3] Clip Triangle Min/Max"=<0 0>
       "[3] Random Seed"=0
      ))
    )
    Create Color Attribute=true
    Create Exponent Attribute=true
    Create Position Attribute=true
    Create Radius Attribute=true
    Exponent=2
    Position:NiagaraPosition=$Particles.Position
    Radius=40
   )
  ]
  rendererClasses [&NiagaraLightRendererProperties]
  renderers [{rendererClass=&NiagaraLightRendererProperties, rendererIndex=0}]
 }
 SpeedLines {
  emitterUpdateScript [
   &EmitterState()
   &SpawnRate(SpawnRate=10, Use Spawn Probability=false)
  ]
  particleSpawnScript [
   /Niagara/Modules/Spawn/Initialization/V2/InitializeParticle(
    Color=<rgba 12 12 12 10>
    Lifetime=0.5
    Sprite Size Max=<2 2>
    Sprite Size Min=<1 1>
    Sprite Size Mode=::NewEnumerator2
    UsePositionOffset=false
    Write Material Random=true
   )
   /SetVariables(Particles.Position=&Add_Vector(A=Linked() {linkedVariable {name=Engine.Owner.Position, type=NiagaraPosition}}, B=<0 0 5>)) {isSetParameters=true, module=SetVariables_#4}
   &TorusLocation(Align Torus With Axis=false, Handle Radius=0, Large Radius=16, MaskSpawnGroup=false, Module.Override Randomness=false, Offset=<0 0 10>)
  ]
  particleUpdateScript [
   &ParticleState()
   &InheritSourceMovement(
    Applied Force Scale=<-0.5 -0.5 0>
    ApplyForce=true
    ApplyForceFalloff=false
    ApplyPosition=true
    ApplyPositionFalloff=false
    Module.LimitForceSpeed=false
   )
   &AddVelocityFromPoint(
    Module.Use Falloff Distance=false
    Origin Offset=<0 0 22>
    Velocity Strength=&FloatFromCurve(
     FloatCurve=DataInterface() {propertyValues {Curve {keys [
         {arriveTangent=-1.0543, interpMode=RCIM_Cubic, time=0.25, value=1}
         {arriveTangent=-5, interpMode=RCIM_Cubic, time=0.6, value=0}
         <at 0.7 -1 RCIM_Cubic>
        ]}}}
     Scale Curve=13
    )
   ) {enabled=false}
   &PointForce(
    Force Falloff Distance=160
    Force Origin=&Add_Vector(A=/Niagara/DynamicInputs/Helpers/SimulationPosition(), B=<0 0 20>)
    Force Strength=1000
    Module.Use Falloff=true
    Module.UseLinearForce=false
   )
   &GravityForce(Gravity=&MakeVector(
     X=0
     Y=0
     Z=&FloatFromCurve(
      FloatCurve=DataInterface() {propertyValues {Curve {keys [<at 0.4 0 RCIM_Cubic>, {arriveTangent=0.5303, interpMode=RCIM_Cubic, leaveTangent=0.5303, tangentMode=RCTM_User, time=0.5, value=1}]}}}
      Scale Curve=3000
     )
    ))
   &SolveForcesAndVelocity(Clamp Velocity=false, Limit Acceleration=false)
   &ScaleColor(
    Scale Alpha=&FloatFromCurve(FloatCurve=DataInterface() {propertyValues {Curve {keys [
         {arriveTangent=2.9808, interpMode=RCIM_Cubic, leaveTangent=2.9808, tangentMode=RCTM_User, time=0, value=0}
         <at 0.5 1 RCIM_Cubic>
         {arriveTangent=-1.6379, interpMode=RCIM_Cubic, leaveTangent=-1.6379, tangentMode=RCTM_User, time=1, value=0}
        ]}}})
    Scale RGB=&MakeVectorFromLinearColorRGB(Color=&ColorFromCurve(ColorCurve=DataInterface() {propertyValues {
        AlphaCurve {keys [<at 0.0077 0.2>, <at 0.0803 1>, {arriveTangent=-1.0873, interpMode=RCIM_Cubic, leaveTangent=-1.0873, time=0.2278, value=1}, <at 1 0>]}
        BlueCurve {keys [<at 0.2688 1>,<at 0.6351 1>]}
        GreenCurve {keys [<at 0.2688 0.889>,<at 0.6351 0>]}
        RedCurve {keys [<at 0.2688 0>,<at 0.6351 0.1722>]}
       }}))
    ScaleA=true
    ScaleRGB=true
   )
   &ScaleSpriteSize(Scale Factor=&Vector2DFromFloat(Value=&FloatFromCurve(
      FloatCurve=DataInterface() {propertyValues {Curve {keys [{arriveTangent=2.4681, interpMode=RCIM_Cubic, leaveTangent=2.4681, tangentMode=RCTM_User, time=0, value=0}, <at 0.5 1>, <at 1 1 RCIM_Cubic>]}}}
      Scale Curve=5
     )))
  ]
  rendererClasses [&NiagaraSpriteRendererProperties]
  renderers [{rendererClass=&NiagaraSpriteRendererProperties}]
 }
 Thunder_Center {
  emitterUpdateScript [
   &EmitterState()
   &SpawnRate(SpawnRate=50, Use Spawn Probability=false)
  ]
  particleSpawnScript [
   &InitializeParticle(
    Color=<rgba 4 4 4 1>
    Lifetime=&RandomRangeFloat(Minimum=0.3, Maximum=0.4)
    Sprite Size Max=<30 50>
    Sprite Size Min=<15 30>
    Sprite Size Mode=::NewEnumerator2
    UsePositionOffset=false
    Write Material Random=true
   )
   /SetVariables(Particles.Position=&Add_Vector(A=Linked() {linkedVariable {name=Engine.Owner.Position, type=NiagaraPosition}}, B=<0 0 5>)) {isSetParameters=true, module=SetVariables_#4}
   &TorusLocation(Align Torus With Axis=false, Handle Radius=0, Large Radius=15, MaskSpawnGroup=false, Module.Override Randomness=false)
   /SetVariables(Particles.RandomFrame=&RandomRangeInt(Minimum=0, Maximum=63)) {isSetParameters=true}
  ]
  particleUpdateScript [
   &ParticleState()
   &InheritSourceMovement(
    Applied Force Scale=<-0.5 -0.5 0>
    ApplyForce=true
    ApplyForceFalloff=false
    ApplyPosition=true
    ApplyPositionFalloff=false
    Module.LimitForceSpeed=false
   ) {module=InheritSourceMovement001}
   &AddVelocityFromPoint(
    Module.Use Falloff Distance=false
    Origin Offset=<0 0 22>
    Velocity Strength=&FloatFromCurve(
     FloatCurve=DataInterface() {propertyValues {Curve {keys [
         {arriveTangent=-1.0543, interpMode=RCIM_Cubic, time=0.25, value=1}
         {arriveTangent=-5, interpMode=RCIM_Cubic, time=0.6, value=0}
         <at 0.7 -1 RCIM_Cubic>
        ]}}}
     Scale Curve=13
    )
   ) {enabled=false}
   &PointForce(
    Force Falloff Distance=160
    Force Origin=&Add_Vector(A=&SimulationPosition(), B=<0 0 22>)
    Force Strength=1000
    Module.Use Falloff=true
    Module.UseLinearForce=false
   )
   &GravityForce(Gravity=&MakeVector(
     X=0
     Y=0
     Z=&FloatFromCurve(
      FloatCurve=DataInterface() {propertyValues {Curve {keys [<at 0.4 0 RCIM_Cubic>, {arriveTangent=0.5303, interpMode=RCIM_Cubic, leaveTangent=0.5303, tangentMode=RCTM_User, time=0.5, value=1}]}}}
      Scale Curve=3000
     )
    ))
   &SolveForcesAndVelocity(Clamp Velocity=false, Limit Acceleration=false)
   &ScaleColor(
    Scale Alpha=&FloatFromCurve(
     FloatCurve=DataInterface() {propertyValues {Curve {keys [
         {arriveTangent=2.9808, interpMode=RCIM_Cubic, leaveTangent=2.9808, tangentMode=RCTM_User, time=0, value=0}
         {arriveTangent=1.2857, interpMode=RCIM_Cubic, leaveTangent=1.2857, time=0.2, value=0.9}
         <at 0.7 0.9>
         {arriveTangent=-1.6379, interpMode=RCIM_Cubic, leaveTangent=-1.6379, tangentMode=RCTM_User, time=1, value=0}
        ]}}}
     Scale Curve=2
    )
    Scale RGB=&MakeVectorFromLinearColorRGB(Color=&ColorFromCurve(ColorCurve=DataInterface() {propertyValues {
        AlphaCurve {keys [<at 0 1 RCIM_Cubic>,<at 0.2068 1>,<at 0.6641 0>]}
        BlueCurve {keys [<at 0.0463 5>,<at 0.2509 1>]}
        GreenCurve {keys [<at 0.0463 2.8913>,<at 0.2509 0.0693>]}
        RedCurve {keys [<at 0.0463 0>,<at 0.2509 0.3306>]}
       }}))
    ScaleA=true
    ScaleRGB=true
   )
   &ScaleSpriteSize(Scale Factor=&Vector2DFromFloat(Value=&FloatFromCurve(
      FloatCurve=DataInterface() {propertyValues {Curve {keys [{arriveTangent=0.553, interpMode=RCIM_Cubic, leaveTangent=0.553, tangentMode=RCTM_User, time=0, value=0}, <at 1 1 RCIM_Cubic>]}}}
      Scale Curve=3.5
     )))
   &FN_SubUVAnimation(End Frame=&Add_Integer(A=$Particles.RandomFrame, B=24), Start Frame=$Particles.RandomFrame, SubUV Loop Count=2)
  ]
  rendererClasses [&NiagaraSpriteRendererProperties]
  renderers [{rendererClass=&NiagaraSpriteRendererProperties}]
 }
}
summary {
 systemName=NS_Skydive_Fuchsia
 emitters [
  {emitterName=SpeedLines, rendererClasses [&NiagaraSpriteRendererProperties]}
  {emitterName=Thunder_Center, rendererClasses [&NiagaraSpriteRendererProperties]}
  {emitterName=Bowl_Multi, rendererClasses [&NiagaraMeshRendererProperties]}
  {emitterName=ElectricMeshArcsBody, rendererClasses [&NiagaraMeshRendererProperties]}
  {emitterName=ElectricMeshArcsUnder002, rendererClasses [&NiagaraMeshRendererProperties]}
  {emitterName=ElectricMeshArcsUnder001, rendererClasses [&NiagaraMeshRendererProperties]}
  {bEnabled=false, emitterName=ElectricMeshArcsUnder003, rendererClasses [&NiagaraMeshRendererProperties]}
  {emitterName=ElectricMeshArcsBody001, rendererClasses [&NiagaraMeshRendererProperties]}
  {emitterName=Lig, rendererClasses [&NiagaraLightRendererProperties]}
  {emitterName=Decals, rendererClasses [&NiagaraSpriteRendererProperties]}
  {bEnabled=false, emitterName=ElectricMeshArcsUnder, rendererClasses [&NiagaraMeshRendererProperties]}
  {bEnabled=false, emitterName=ElectricMeshArcsUnder004, rendererClasses [&NiagaraMeshRendererProperties]}
 ]
 userVariables [
  User.bIsFrontEnd {type=NiagaraBool}
  User.bIsFrontEndPreview {type=NiagaraBool}
  User.Color_Tint {type=LinearColor, description="", defaultValue {struct=LinearColor, value=<rgba 1 1 1 1>}}
  User.EffectIntensity {type=NiagaraFloat, defaultValue=0, description=""}
  User.ParaGlideLeanAlpha {type=NiagaraFloat, defaultValue=0, description=""}
  User.PlayerSpeed {type=NiagaraFloat, defaultValue=0, description=""}
  User.RibbonIDOffset {type=NiagaraInt32}
  User.SkeletalMesh {
   type=NiagaraDataInterfaceSkeletalMesh
   defaultValue {
    struct=NiagaraExt_VariableValue_DataInterface
    value {dataInterface=&self:NiagaraDataInterfaceSkeletalMesh_0}
   }
  }
 ]
}
